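# Sound cue for a balloon pop, named for quest use in the Figment environment.
# A Random node picks one of five equally weighted balloon pop waves, then a Modulator slightly varies the pitch on each play.
# It plays in the gadgets sound class with gadget-activate attenuation and overrides concurrency to block new instances at the limit rather than stop playing ones.
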SoundCue /FigmentEnvironment/Sounds/Fort_Gadgets/Balloons/Balloons_Quest_Pop_Cue
root=Modulator_0
nodes [
 {id=Modulator_0, children [Random_0], properties {pitchMax=1.1, pitchMin=0.9}}
 {id=Random_0, children [WavePlayer_0,WavePlayer_2,WavePlayer_3,WavePlayer_4,WavePlayer_1], properties {weights [1,1,1,1,1]}}
 {id=WavePlayer_0, properties {soundWaveAssetPtr=/Game/Sounds/Fort_Gadgets/Balloons/Balloon_Pop_01}}
 {id=WavePlayer_1, properties {soundWaveAssetPtr=/Game/Sounds/Fort_Gadgets/Balloons/Balloon_Pop_05}}
 {id=WavePlayer_2, properties {soundWaveAssetPtr=/Game/Sounds/Fort_Gadgets/Balloons/Balloon_Pop_02}}
 {id=WavePlayer_3, properties {soundWaveAssetPtr=/Game/Sounds/Fort_Gadgets/Balloons/Balloon_Pop_03}}
 {id=WavePlayer_4, properties {soundWaveAssetPtr=/Game/Sounds/Fort_Gadgets/Balloons/Balloon_Pop_04}}
]
properties {
 attenuationSettings=&Fort_GadgetActivate_Att
 bOverrideConcurrency=true
 maxDistance=4000
 soundClassObject=&Fort_Gadgets
 volumeMultiplier=1
 concurrencyOverrides {resolutionRule=PreventNew}
}
